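# Sound cue for opening the Seven faction chest, judging by its name, in the loot reward sound class.
# A single wave player feeds the open sound straight to the output with boosted volume, with no randomization or mixing.
# Spatialization comes from the shared item-interact attenuation asset; inline attenuation values are stored, but their override switch is not set.
# A concurrency override blocks new instances once the limit is reached.
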SoundCue /FlipperGameplay/Items/TheSevenChest/Audio/Faction_Chest_Seven_Open_Cue
root=WavePlayer_3
nodes [{id=WavePlayer_3, properties {soundWaveAssetPtr=&/Faction_Chest_Seven_Open_01}}]
properties {
 attenuationSettings=/Game/Sounds/Attenuation/Fort_Item_Interact_Att
 bOverrideConcurrency=true
 maxDistance=1850
 soundClassObject=&Fort_Loot_Reward_PSM
 volumeMultiplier=1.8
 attenuationOverrides {bEnableOcclusion=true, falloffDistance=1000, occlusionTraceChannel=ECC_GameTraceChannel11, attenuationShapeExtents {x=500}}
 concurrencyOverrides {resolutionRule=PreventNew}
}
