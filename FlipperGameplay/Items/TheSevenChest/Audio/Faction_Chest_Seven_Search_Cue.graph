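# Sound cue for searching The Seven faction chest, playing what is named as the search-start sound.
# A single wave player feeds the output directly, with no randomization or mixing.
# It uses the shared mono item-interact attenuation and the loot-collect sound class, with a boosted volume.
# A concurrency override blocks new instances once its limit is reached rather than cutting off playing ones.
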SoundCue /FlipperGameplay/Items/TheSevenChest/Audio/Faction_Chest_Seven_Search_Cue
root=WavePlayer_5
nodes [{id=WavePlayer_5, properties {soundWaveAssetPtr=&/Faction_Chest_Seven_Search_Start_01}}]
properties {
 attenuationSettings=/Game/Sounds/Attenuation/Fort_Item_Interact_Mono_Att
 bOverrideConcurrency=true
 maxDistance=1500
 soundClassObject=&Fort_Loot_Collect
 volumeMultiplier=1.25
 concurrencyOverrides {resolutionRule=PreventNew}
}
